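# Animation Blueprint for the Professor Pup cosmetic body, a mech with a wheel and a pup rider, driven by the parent mesh's copied pose.
# Springs, Control Rigs, wheel rotation and belt/plate dynamics build the mech pose, then pup head constraints and additive idle follow.
# Synced montages or traversal emotes replace that branch via per-bone layering, before tail dynamics.
# Events broadcast brake lights and set hide_hands, whose arm-shrinking nodes are unconnected to output.
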AnimBlueprint /BRCosmetics/Characters/Player/Male/Large/Bodies/M_LRG_Professor_Pup/Meshes/M_LRG_Professor_Pup_AnimBP
target_skeleton=&/M_LRG_Professor_Pup_Skeleton
graphs [
 AnimGraph {nodes [
   n1 = Get hide_hands() -> (hide_hands:double)
   n2 = "Transform (Modify) Bone - Bone: lowerarm_l_start"(Alpha=@n1, Translation="0,0,0", Rotation="0,0,0", Scale="0.01,0.01,0.01") {
    rotationSpace=BCS_BoneSpace
    scaleMode=BMM_Replace
    scaleSpace=BCS_BoneSpace
    alphaBoolBlend {blendOption=Linear}
    boneToModify {boneName=lowerarm_l_start}
    scale=<0.01 0.01 0.01>
   }
   "Transform (Modify) Bone - Bone: lowerarm_r_start"(ComponentPose=@n2, Alpha=@n1, Translation="0,0,0", Rotation="0,0,0", Scale="0.01,0.01,0.01") {
    rotationSpace=BCS_BoneSpace
    scaleMode=BMM_Replace
    scaleSpace=BCS_BoneSpace
    alphaBoolBlend {blendOption=Linear}
    boneToModify {boneName=lowerarm_r_start}
    scale=<0.01 0.01 0.01>
   }
   n3 = Get SkydiveReduceAlpha() -> (SkydiveReduceAlpha:float)
   n4 = "Constraint - p_neck_01 (R)"(ComponentPose=Local To Component(LocalPose=Use cached pose 'MechPose'()), Alpha=@n3, ConstraintWeights_0=1) {
    alphaBoolBlend {blendOption=Linear}
    alphaScaleBiasClamp {scale=0.4}
    boneToModify {boneName=p_neck_01}
    constraintSetup [{transformType=Rotation, targetBone {boneName=neck_02}}]
    constraintWeights [1]
   }
   n5 = "Constraint - p_head (R)"(ComponentPose=@n4, Alpha=@n3, ConstraintWeights_0=1) {
    alphaBoolBlend {blendOption=Linear}
    alphaScaleBiasClamp {scale=0.4}
    boneToModify {boneName=p_head}
    constraintSetup [{transformType=Rotation, targetBone {boneName=head}}]
    constraintWeights [1]
   }
   n6 = Apply Additive(
    Base=Component To Local(ComponentPose=@n5)
    Additive="Sequence Player (additive) 'professoepup_Idle'"() {sequence=/BRCosmetics/Characters/Player/Male/Large/Heads/M_LRG_Professor_Pup/professoepup_Idle}
    Alpha=Get AdditiveIdleAlpha() -> (AdditiveIdleAlpha:float)
   ) {alphaBoolBlend {blendOption=Linear}}
   n7 = Get SyncedSequence() -> (SyncedSequence:Anim Sequence Object Reference)
   n8 = Get SyncedMontagePosition() -> (SyncedMontagePosition:float)
   n9 = Get TraversalEmoteTime() -> (TraversalEmoteTime:float)
   n10 = Layered blend per bone(
    BasePose=Use cached pose 'CopiedPose'()
    BlendPoses_0=Blend Poses by bool(
     BlendPose_0="Sequence Evaluator (by time) 'emote_prof_pup_traversal_PUP'"!(ExplicitTime:float=@n9) -> (Pose:Pose Link Structure) {is=title AnimGraphNode_SequenceEvaluator, sequence=/BRCosmetics/Animation/Game/MainPlayer/Emotes/ProfessorPup/emote_prof_pup_traversal_PUP}
     BlendPose_1="Sequence Evaluator (by time)"!(Sequence:Anim Sequence Base Object Reference=@n7, ExplicitTime:float=@n8) -> (Pose:Pose Link Structure) {is=title AnimGraphNode_SequenceEvaluator}
     BlendTime_0=0
     BlendTime_1=0
     bActiveValue=Get bIsBlendingTraversalEmote() -> (bIsBlendingTraversalEmote:bool)
    ) {blendType=Linear}
   ) {layerSetup [{branchFilters [{boneName=cyclePivot_l}, {boneName=pelvis_shift}]}]}
   n11 = Get bIsPlayingSyncedMontage() -> (bIsPlayingSyncedMontage:bool)
   n12 = Get DynamicsAlpha() -> (DynamicsAlpha:float)
   n13 = Anim Dynamics - Bone: p_tail_01(
    ComponentPose=Local To Component(LocalPose=Blend Poses by bool(BlendPose_0=@n10, BlendPose_1=@n6, bActiveValue=@n11) {blendType=Linear})
    Alpha=@n12
    ExternalForce="0,0,0"
   ) {
    angularDampingOverride=1
    angularSpringConstant=1500
    bAngularSpring=true
    bGravityOverrideInSimSpace=true
    bOverrideAngularDamping=true
    gravityScale=0.15
    simulationSpace=RootRelative
    alphaBoolBlend {blendOption=Linear}
    alphaScaleBiasClamp {bInterpResult=true, interpSpeedDecreasing=5, interpSpeedIncreasing=5, scale=0.5}
    boundBone {boneName=p_tail_01}
    componentLinearAccScale=<0.4 0.4 0.45>
    componentLinearVelScale=<0.4 0.4 0.45>
    physicsBodyDefinitions [{
      sphereCollisionRadius=0
      boundBone {boneName=p_tail_01}
      boxExtents=<10 20 20>
      constraintSetup {angularLimitsMax=<0 80 70>, angularLimitsMin=<0 -80 -70>, angularTarget=<0 0 0>}
      localJointOffset=<10 0 0>
     }]
   }
   n14 = Anim Dynamics - Bone: p_tail_02(ComponentPose=@n13, Alpha=@n12, ExternalForce="0,0,0") {
    angularDampingOverride=1
    angularSpringConstant=1500
    bAngularSpring=true
    bGravityOverrideInSimSpace=true
    bOverrideAngularDamping=true
    gravityScale=0.15
    simulationSpace=RootRelative
    alphaBoolBlend {blendOption=Linear}
    alphaScaleBiasClamp {bInterpResult=true, interpSpeedDecreasing=5, interpSpeedIncreasing=5, scale=0.5}
    boundBone {boneName=p_tail_02}
    componentLinearAccScale=<0.4 0.4 0.46>
    componentLinearVelScale=<0.4 0.4 0.46>
    physicsBodyDefinitions [{
      sphereCollisionRadius=0
      boundBone {boneName=p_tail_02}
      boxExtents=<10 20 20>
      constraintSetup {angularLimitsMax=<0 80 70>, angularLimitsMin=<0 -80 -70>, angularTarget=<0 0 0>}
      localJointOffset=<10 0 0>
     }]
   }
   n15 = Anim Dynamics - Bone: p_tail_03(ComponentPose=@n14, Alpha=@n12, ExternalForce="0,0,0") {
    angularDampingOverride=1
    angularSpringConstant=1500
    bAngularSpring=true
    bGravityOverrideInSimSpace=true
    bOverrideAngularDamping=true
    gravityScale=0.15
    simulationSpace=RootRelative
    alphaBoolBlend {blendOption=Linear}
    alphaScaleBiasClamp {bInterpResult=true, interpSpeedDecreasing=5, interpSpeedIncreasing=5, scale=0.5}
    boundBone {boneName=p_tail_03}
    componentLinearAccScale=<0.4 0.4 0.47>
    componentLinearVelScale=<0.4 0.4 0.47>
    physicsBodyDefinitions [{
      sphereCollisionRadius=0
      boundBone {boneName=p_tail_03}
      boxExtents=<10 20 20>
      constraintSetup {angularLimitsMax=<0 80 70>, angularLimitsMin=<0 -80 -70>, angularTarget=<0 0 0>}
      localJointOffset=<10 0 0>
     }]
   }
   n16 = Anim Dynamics - Bone: p_tail_04(ComponentPose=@n15, Alpha=@n12, ExternalForce="0,0,0") {
    angularDampingOverride=1
    angularSpringConstant=1500
    bAngularSpring=true
    bGravityOverrideInSimSpace=true
    bOverrideAngularDamping=true
    gravityScale=0.15
    simulationSpace=RootRelative
    alphaBoolBlend {blendOption=Linear}
    alphaScaleBiasClamp {bInterpResult=true, interpSpeedDecreasing=5, interpSpeedIncreasing=5, scale=0.5}
    boundBone {boneName=p_tail_04}
    componentLinearAccScale=<0.4 0.4 0.48>
    componentLinearVelScale=<0.4 0.4 0.48>
    physicsBodyDefinitions [{
      sphereCollisionRadius=0
      boundBone {boneName=p_tail_04}
      boxExtents=<10 20 20>
      constraintSetup {angularLimitsMax=<0 80 70>, angularLimitsMin=<0 -80 -70>, angularTarget=<0 0 0>}
      localJointOffset=<10 0 0>
     }]
   }
   n17 = Get ModifyTailAlpha() -> (ModifyTailAlpha:float)
   n18 = "Transform (Modify) Bone - Bone: p_tail_01"(ComponentPose=@n16, Alpha=@n17, Translation="0,0,0", Rotation="90,0,0", Scale="1,1,1") {
    rotationMode=BMM_Replace
    rotationSpace=BCS_BoneSpace
    alphaBoolBlend {blendOption=Linear}
    boneToModify {boneName=p_tail_01}
    rotation {pitch=90}
   }
   n19 = "Transform (Modify) Bone - Bone: p_tail_02"(ComponentPose=@n18, Alpha=@n17, Translation="0,0,0", Rotation="-30,-60,0", Scale="1,1,1") {
    rotationMode=BMM_Replace
    rotationSpace=BCS_BoneSpace
    alphaBoolBlend {blendOption=Linear}
    boneToModify {boneName=p_tail_02}
    rotation {pitch=-30, yaw=-60}
   }
   Output Pose(Result=Component To Local(ComponentPose=@n19)) {layerGroup=None}
   Save cached pose 'CopiedPose'(Pose=Copy Pose From Mesh() {bCopyCurves=true, bUseAttachedParent=true}) {cacheName=CopiedPose}
   n20 = Control Rig(Source=Use cached pose 'PoseA'(), Alpha=Get MechBodyAlpha() -> (MechBodyAlpha:float)) {
    alphaBoolBlend {blendOption=Linear}
    controlRigAssetReference {blueprintRigClass=&/Torso_Redux.Torso_Redux_C}
   }
   n21 = Use cached pose 'PoseA'()
   n22 = Get bIsOnGrindRail()
   n23 = "Transform (Modify) Bone - Bone: wheel"!(
    ComponentPose=Local To Component(LocalPose=Blend Poses by bool(BlendPose_0=@n21, BlendPose_1=@n20, bActiveValue=@n22) {blendType=Linear})
    Alpha=Get CatBodyAlpha() -> (CatBodyAlpha:float)
    Rotation=Get WheelRotator() -> (WheelRotator:Rotator)
   ) -> (Pose) {
    rotationMode=BMM_Replace
    rotationSpace=BCS_ParentBoneSpace
    translationSpace=BCS_WorldSpace
    alphaBoolBlend {blendOption=Linear}
    boneToModify {boneName=wheel}
   }
   n24 = Get MechBodyAlpha()
   n25 = Anim Dynamics - Bone: plate(ComponentPose=@n23, Alpha=@n24, ExternalForce="0,0,0") {
    angularDampingOverride=1
    angularSpringConstant=1500
    bAngularSpring=true
    bGravityOverrideInSimSpace=true
    bOverrideAngularDamping=true
    gravityScale=0.35
    simulationSpace=RootRelative
    alphaBoolBlend {blendOption=Linear}
    boundBone {boneName=plate}
    physicsBodyDefinitions [{
      sphereCollisionRadius=0
      boundBone {boneName=plate}
      boxExtents=<10 20 20>
      constraintSetup {angularLimitsMax=<0 0 25>, angularLimitsMin=<0 0 -40>, angularTarget=<0 0 0>}
      localJointOffset=<10 0 0>
     }]
   }
   n26 = Anim Dynamics - Bone: dyn_thigh_belt_r(ComponentPose=@n25, Alpha=@n24, ExternalForce="0,0,0") {
    angularDampingOverride=1
    angularSpringConstant=1500
    bAngularSpring=true
    bGravityOverrideInSimSpace=true
    bOverrideAngularDamping=true
    gravityScale=0.35
    simulationSpace=RootRelative
    alphaBoolBlend {blendOption=Linear}
    alphaScaleBiasClamp {scale=0.3}
    boundBone {boneName=dyn_thigh_belt_r}
    physicsBodyDefinitions [{
      sphereCollisionRadius=0
      boundBone {boneName=dyn_thigh_belt_r}
      boxExtents=<11 22 18>
      constraintSetup {angularLimitsMax=<0 90 90>, angularLimitsMin=<0 -90 -90>, angularTarget=<0 0 0>}
      localJointOffset=<8 0 0>
     }]
   }
   n27 = Anim Dynamics - Bone: dyn_thigh_belt_l(ComponentPose=@n26, Alpha=@n24, ExternalForce="0,0,0") {
    angularDampingOverride=1
    angularSpringConstant=1500
    bAngularSpring=true
    bGravityOverrideInSimSpace=true
    bOverrideAngularDamping=true
    gravityScale=0.35
    simulationSpace=RootRelative
    alphaBoolBlend {blendOption=Linear}
    alphaScaleBiasClamp {scale=0.3}
    boundBone {boneName=dyn_thigh_belt_l}
    physicsBodyDefinitions [{
      sphereCollisionRadius=0
      boundBone {boneName=dyn_thigh_belt_l}
      boxExtents=<10 20 20>
      constraintSetup {angularLimitsMax=<0 90 90>, angularLimitsMin=<0 -90 -90>, angularTarget=<0 0 0>}
      localJointOffset=<10 0 0>
     }]
   }
   Save cached pose 'MechPose'(Pose=Component To Local(ComponentPose=@n27)) {cacheName=MechPose}
   n28 = Get VehicleOrSkydivingOffAlpha() -> (VehicleOrSkydivingOffAlpha:float)
   n29 = Spring controller - Bone: cyclePivot_proxy!(ComponentPose:Component Space Pose Link Structure=Local To Component(LocalPose=Use cached pose 'CopiedPose'()), Alpha:float=@n28) -> (Pose:Component Space Pose Link Structure) {
    is=title AnimGraphNode_SpringBone
    maxDisplacement=10
    springDamping=10
    springStiffness=200
    alphaBoolBlend {blendOption=Linear}
    alphaScaleBiasClamp {scale=0.25}
    springBone {boneName=cyclePivot_proxy}
   }
   n30 = Spring controller - Bone: head!(ComponentPose:Component Space Pose Link Structure=@n29, Alpha:float=@n28) -> (Pose:Component Space Pose Link Structure) {
    is=title AnimGraphNode_SpringBone
    maxDisplacement=10
    springDamping=20
    springStiffness=150
    alphaBoolBlend {blendOption=Linear}
    alphaScaleBiasClamp {scale=0.25}
    springBone {boneName=head}
   }
   n31 = Control Rig!(Source=Component To Local(ComponentPose=@n30), Alpha, Pub_IsOnGrindRail:bool=Get bIsOnGrindRail()) -> (Pose) {
    alphaBoolBlend {blendOption=Linear}
    controlRigAssetReference {blueprintRigClass=&/Torso_Only_CtrlRig.Torso_Only_CtrlRig_C}
   }
   n32 = Control Rig!(Source=@n31, Alpha=Get MechBodyAlpha(), CurrentLOD_float:double=Get CurrentLODValue() -> (CurrentLODValue:float)) -> (Pose) {
    alphaBoolBlend {blendOption=Linear}
    controlRigAssetReference {blueprintRigClass=&/Legs_Only_CtrlRig.Legs_Only_CtrlRig_C}
   }
   Save cached pose 'PoseA'(Pose=@n32) {cacheName=PoseA}
   SelectFloat(A=0, B=1, bPickA="float > float"(A=Get PawnSpeed2D(), B=800) {is=title K2Node_CallFunction})
  ]}
 EventGraph {nodes [
   Event BlueprintUpdateAnimation()
   > n1 = Sequence()
   > Set ForGrindRail(ForGrindRail:bool=false) -> (Output_Get:bool)
   Call Lights!(
    execute:Exec=@n1.then_1
    self:M LRG Professor Pup Anim BP Object Reference
    BrakeLight:bool=Get bTurnOnBreaklight() -> (bTurnOnBreaklight:bool)
   ) -> (then:Exec) {is=title K2Node_CallDelegate}
   Event OnWeaponChanged()
   > n2 = Cast To FortPawn!(execute, Object=GetOwningActor()) -> (then, CastFailed, AsFort Pawn:Fort Pawn Object Reference) {is=title K2Node_DynamicCast}
   n3 = Get CurrentWeapon!(self:Fort Pawn Object Reference=@n2) -> (CurrentWeapon:Fort Weapon Object Reference)
   > Set hide_hands(hide_hands:double="To Float (Boolean)"(InBool=WeaponDataHasAnyMatchingGameplayTags!(
      self:Fort Weapon Object Reference=@n3
      TagContainer:Gameplay Tag Container Structure=(GameplayTags=((TagName="Animation.CharacterParts.HideFX.Hands")))
     ) -> (ReturnValue:bool))) -> (Output_Get:double)
   "float > float"(A=Get PawnSpeed2D(), B=0)
   AND Boolean(A=Get bIsOnGrindRail(), B=false)
   Get MechBodyAlpha()
  ]}
 Lights {nodes [Lights!() -> (then, BrakeLight:bool=~) {is=title K2Node_FunctionEntry}]}
]
variables [
 ForGrindRail {type=boolean}
 hide_hands {type=number}
 Lights {default=()}
]
