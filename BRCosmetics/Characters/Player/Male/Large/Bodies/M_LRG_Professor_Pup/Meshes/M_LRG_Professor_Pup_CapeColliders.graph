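# Physics asset of cape colliders for the large male Professor Pup body, giving the character's cape simulation shapes to collide against.
# Hand-fitted capsules cover the pelvis, spine, clavicles, hands, left forearm, and the plate and seat bones, with a sphere on the head.
# Leg, foot, upper-arm, right-forearm and neck bodies carry no shapes.
# It defines no constraints, so it provides collision volumes only, without ragdoll joints.
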PhysicsAsset /BRCosmetics/Characters/Player/Male/Large/Bodies/M_LRG_Professor_Pup/Meshes/M_LRG_Professor_Pup_CapeColliders
bodies {
 calf_l []
 calf_r []
 clavicle_l [{shapeType=Capsule, length=11.3526, radius=12.6279, shapeName=clavicle_l_capsule, center=<10.2413 -1.5523 -1.15>, rotation=<rot 62.6575 0 0>}]
 clavicle_r [
  {
   shapeType=Capsule
   length=10.4776
   radius=13.5029
   shapeName=clavicle_l_capsule
   center=<-13.6978 1.5519 2.1897>
   rotation=<rot 73.2583 0 180>
  }
  {
   shapeType=Capsule
   length=4.2188
   radius=3.1562
   shapeName=clavicle_r_capsule
   center=<-21.3867 10.5114 -5.3012>
   rotation=<rot -9.8183 20.4149 -65.3852>
  }
 ]
 foot_l []
 foot_r []
 hand_l [{shapeType=Capsule, length=5.0012, radius=7.2893, shapeName=hand_l_capsule, center=<7.1732 -2.8011 -0.0005>, rotation=<rot 90 0 0>}]
 hand_r [{shapeType=Capsule, length=5.0012, radius=7.2893, shapeName=hand_l_capsule, center=<-7.1732 2.8011 0.0005>, rotation=<rot 90 0 180>}]
 head [{shapeType=Sphere, radius=11.4375, center=<6.0918 2.3775 -0.0001>}]
 lowerarm_l [{
   shapeType=Capsule
   length=18.9375
   radius=4.2812
   shapeName=lowerarm_l_capsule
   center=<11.3123 -5.6225 5.8608>
   rotation=<rot 63.0403 -37.999 -39.9565>
  }]
 lowerarm_r []
 neck_02 []
 pelvis [
  {shapeType=Capsule, length=12.7645, radius=13.3293, center=<-4.8511 0 0>}
  {
   shapeType=Capsule
   length=11.8125
   radius=7.7812
   shapeName=pelvis_capsule
   center=<-6.657 -7.0997 -12.7562>
   rotation=<rot 12.4066 36.3045 77.6229>
  }
  {
   shapeType=Capsule
   length=11.8125
   radius=7.7812
   shapeName=pelvis_capsule
   center=<-6.7653 -6.9662 12.7194>
   rotation=<rot -12.4066 36.3045 -77.623>
  }
 ]
 plate [{
   shapeType=Capsule
   length=13.3809
   radius=6.888
   shapeName=plate_capsule
   center=<3.3437 4.8388 -0.0009>
   rotation=<rot -0.0082 -89.9999 0.0283>
  }]
 seat [{shapeType=Capsule, length=4.0312, radius=11.4375, shapeName=spine_04_capsule, center=<-1.7885 -15.0291 8.2168>, rotation=<rot 90 0 0>}]
 spine_01 [{shapeType=Capsule, length=0, radius=14.0564, center=<0 1.3296 0>}]
 spine_02 [{shapeType=Capsule, length=10.5625, radius=15.2439, center=<0 0.4678 -0.0001>}]
 spine_03 [{shapeType=Capsule, length=5, radius=18.4939, center=<-1.8049 1.268 0>}]
 spine_04 [{shapeType=Capsule, length=1.5694, radius=17.4682, center=<-0.0424 -1.5349 0.0001>, rotation=<rot 0 3.3589 0>}]
 spine_05 [
  {shapeType=Capsule, length=14.3855, radius=15.242, center=<0.0615 -4.4055 0.0057>}
  {shapeType=Capsule, length=16.375, radius=8.9375, shapeName=spine_05_capsule, center=<11.7331 -11.5577 0>}
 ]
 thigh_l []
 thigh_r []
 upperarm_l []
 upperarm_r []
}
constraints []
