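# Anim Blueprint for the Professor Pup large male face accessory, which follows the attached parent mesh and adds ear physics.
# It copies the parent's pose and curves, then layers the Pup idle from faceAttach down, weighted by the DisableFaceOverride curve.
# Separately tuned Anim Dynamics nodes move the left and right ears.
# Each update reads head scale from the customization base mesh and applies it to p_head.
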AnimBlueprint /BRCosmetics/Characters/Player/Male/Large/Bodies/M_LRG_Professor_Pup/Meshes/Parts/M_LRG_Professor_Pup_FaceAcc_AnimBP
target_skeleton=&/M_LRG_Professor_Pup_FaceAcc_Skeleton
graphs [
 AnimGraph {nodes [
   n1 = Copy Pose From Mesh() {bCopyCurves=true, bUseAttachedParent=true}
   n2 = Sequence Player 'M_LRG_Professor_Pup_Idle'() {sequence=&/M_LRG_Professor_Pup_Idle}
   n3 = float * float(A=1, B=GetCurveValue(CurveName=DisableFaceOverride))
   n4 = Anim Dynamics - Bone: dyn_ear_l(
    ComponentPose=Local To Component(LocalPose=Layered blend per bone(BasePose=@n1, BlendPoses_0=@n2, BlendWeights_0=@n3) {layerSetup [{branchFilters [{boneName=faceAttach}]}]})
    Alpha=1
    ExternalForce="0,0,0"
   ) {
    bGravityOverrideInSimSpace=true
    bUseGravityOverride=true
    simulationSpace=Actor
    alphaBoolBlend {blendOption=Linear}
    boundBone {boneName=dyn_ear_l}
    componentAppliedLinearAccClamp=<2000 2000 2000>
    componentLinearAccScale=<0.5 0.5 0.5>
    componentLinearVelScale=<0.5 0.5 0.5>
    gravityOverride=<1 1 0.85>
    physicsBodyDefinitions [{
      sphereCollisionRadius=0
      boundBone {boneName=dyn_ear_l}
      boxExtents=<6 6 2>
      constraintSetup {angularLimitsMax=<0 25 0>, angularLimitsMin=<0 1 0>, angularTarget=<0 0 0>}
      localJointOffset=<2 0 0>
     }]
    relativeSpaceBone {boneName=p_spine_05}
   }
   n5 = Anim Dynamics - Bone: dyn_ear_r(ComponentPose=@n4, Alpha=0.75, ExternalForce="0,0,0") {
    bGravityOverrideInSimSpace=true
    bUseGravityOverride=true
    alphaBoolBlend {blendOption=Linear}
    boundBone {boneName=dyn_ear_r}
    componentAppliedLinearAccClamp=<2000 2000 2000>
    componentLinearAccScale=<0.25 0.25 0.25>
    componentLinearVelScale=<0.25 0.25 0.25>
    gravityOverride=<1 1 0.75>
    physicsBodyDefinitions [{
      sphereCollisionRadius=0
      boundBone {boneName=dyn_ear_r}
      boxExtents=<6 6 2>
      constraintSetup {angularLimitsMax=<0 3 0>, angularLimitsMin=<0 -15 0>, angularTarget=<0 0 0>}
      localJointOffset=<2 0 0>
     }]
    relativeSpaceBone {boneName=pelvis}
   }
   n6 = Get Head Scale() -> (Head Scale:Vector)
   Output Pose(Result=Component To Local(ComponentPose="Transform (Modify) Bone - Bone: p_head"!(ComponentPose=@n5, Alpha=1, Scale=@n6) -> (Pose) {scaleMode=BMM_Replace, scaleSpace=BCS_BoneSpace, boneToModify {boneName=p_head}})) {layerGroup=None}
  ]}
 EventGraph {nodes [
   n1 = Event OnCharacterCustomizationLoaded()
   > Set Skel Mesh(Skel Mesh=@n1.CustomBaseMesh)
   Event BlueprintUpdateAnimation()
   n2 = Get Skel Mesh()
   > Set Head Scale(Head Scale:Vector=BreakTransform(InTransform=GetBoneTransform!(self:Skinned Mesh Component Object Reference=@n2, InBoneName:Name=head, TransformSpace:ERelativeTransformSpace Enum=RTS_ParentBoneSpace) -> (ReturnValue:Transform)).Scale) -> (Output_Get:Vector)
  ]}
]
variables [
 Head Scale {type=Vector}
 Skel Mesh {type=SkeletalMeshComponent}
]
